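# Animation Blueprint for the WildCard back bling, which plays a card animation when a chest is opened.
# The aie.ChestOpened input event sets a flag that sends the state machine from Idle into a one-shot sequence, returning near its end.
# Additive bone modifiers then offset card_body_MSTR constantly, plus further eased offsets while sprinting or skydiving.
# Procedural card slide and flip functions appear unused: their calling chain is unwired and the AnimGraph ignores their outputs.
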AnimBlueprint /BRCosmetics/Accessories/FORT_Backpacks/Backpack_F_MED_WildCard/Meshes/F_MED_WildCard_Pack_AnimBP
target_skeleton=&/F_MED_WildCard_Pack_Skeleton
graphs [
 AnimGraph {nodes [
   AnimGraphNode_StateMachine_0 = State Machine() {is=title AnimGraphNode_StateMachine}
   n1 = "Transform (Modify) Bone - Bone: card_body_MSTR"!(
    ComponentPose=Local To Component(LocalPose=@AnimGraphNode_StateMachine_0)
    bAlphaBoolEnabled:bool=True
    Translation="0,-4,0"
    Rotation="0,0,0"
    Scale="1,1,1"
   ) -> (Pose) {
    alphaInputType=Bool
    translationMode=BMM_Additive
    alphaBoolBlend {blendOption=ExpInOut}
    boneToModify {boneName=card_body_MSTR}
    translation {y=-4}
   }
   n2 = "Transform (Modify) Bone - Bone: card_body_MSTR"(ComponentPose=@n1, bAlphaBoolEnabled=Get bIsSprinting(), Translation="0,-6,-2", Rotation="0,0,0", Scale="1,1,1") {
    alphaInputType=Bool
    translationMode=BMM_Additive
    alphaBoolBlend {blendInTime=0.45, blendOption=ExpInOut, blendOutTime=0.45}
    boneToModify {boneName=card_body_MSTR}
    translation {y=-6, z=-2}
   }
   n3 = "Transform (Modify) Bone - Bone: card_body_MSTR"(ComponentPose=@n2, bAlphaBoolEnabled=Get bIsSkyDiving(), Translation="0,-8,-5", Rotation="0,0,0", Scale="1,1,1") {
    alphaInputType=Bool
    translationMode=BMM_Additive
    alphaBoolBlend {blendInTime=0.5, blendOption=ExpInOut, blendOutTime=0.5}
    boneToModify {boneName=card_body_MSTR}
    translation {y=-8, z=-5}
   }
   Output Pose(Result=Component To Local(ComponentPose=@n3)) {layerGroup=None}
  ]}
 State Machine {
  entry=Idle
  path=AnimGraph.AnimGraphNode_StateMachine_0.State Machine
  ids=id
  nodes [
   AnimStateNode_0 = Idle()
   AnimStateNode_1 = ChestOpened_anim() {is=title AnimStateNode}
   AnimStateTransitionNode_0 = AnimStateTransitionNode()
   AnimStateTransitionNode_1 = AnimStateTransitionNode() {blendMode=ExpInOut, crossfadeDuration=0.6}
  ]
  links [
   {
    dst=Idle
    rule=AnimGraph.AnimGraphNode_StateMachine_0.State Machine.AnimStateTransitionNode_1.Transition
    src=ChestOpened_anim
    via=AnimStateTransitionNode_1
   }
   {
    dst=ChestOpened_anim
    rule=AnimGraph.AnimGraphNode_StateMachine_0.State Machine.AnimStateTransitionNode_0.Transition
    src=Idle
    via=AnimStateTransitionNode_0
   }
  ]
 }
 Idle {path=AnimGraph.AnimGraphNode_StateMachine_0.State Machine.AnimStateNode_0.Idle, nodes [Output Animation Pose()]}
 ChestOpened_anim {
  path=AnimGraph.AnimGraphNode_StateMachine_0.State Machine.AnimStateNode_1.ChestOpened_anim
  nodes [Output Animation Pose(Result=Sequence Player 'F_Med_WildCard_Pack'() {bLoopAnimation=false, sequence=/BRCosmetics/Accessories/FORT_Backpacks/Backpack_F_MED_WildCard/Animation/F_Med_WildCard_Pack})]
 }
 Transition {
  path=AnimGraph.AnimGraphNode_StateMachine_0.State Machine.AnimStateTransitionNode_0.Transition
  nodes [Result(bCanEnterTransition=Get isChestOpened?() -> (isChestOpened?:bool))]
 }
 Transition {
  path=AnimGraph.AnimGraphNode_StateMachine_0.State Machine.AnimStateTransitionNode_1.Transition
  nodes [Result(bCanEnterTransition="float < float"(A="Time Remaining (ratio) (F_Med_WildCard_Pack)"!() -> (ReturnValue:float) {is=title K2Node_AnimGetter}, B=0.1))]
 }
 CardMovementFunc {nodes [
   n1 = CardMovementFunc!() -> (then, DeltaTime:double=~) {is=title K2Node_FunctionEntry}
   n2 = Get MovementAlpha() -> (MovementAlpha:double)
   > Set MovementAlpha(MovementAlpha:double=float + float(A=float * float(A=@n1, B=0.8) {is=title K2Node_PromotableOperator}, B=@n2)) -> (Output_Get:double)
   > Set CardMovement(CardMovement:Vector=MakeVector(Y=FInterpEaseInOut!(A:double=0, B:double=-25, Alpha:double=@n2, Exponent:double=3) -> (ReturnValue:double))) -> (Output_Get:Vector)
   > Branch(Condition="float >= float"(A=@n2, B=1))
   n3 = Get ReverseMovementAlpha() -> (ReverseMovementAlpha:double)
   > Set ReverseMovementAlpha(ReverseMovementAlpha:double=float + float(A=float * float(A=@n1, B=0.8), B=@n3)) -> (Output_Get:double)
   > Set CardMovement(CardMovement=MakeVector(Y=FInterpEaseInOut(A=-25, B=0, Alpha=@n3, Exponent=3)))
   > Branch(Condition="float >= float"(A=@n3, B=1))
   > Set CardMovement(CardMovement="0, 0, 0")
   > Set MovementAlpha(MovementAlpha=0)
   > Set ReverseMovementAlpha(ReverseMovementAlpha=0)
   > Set isChestOpened?(isChestOpened?:bool=false) -> (Output_Get:bool)
  ]}
 CardRotationFunc {nodes [
   n1 = CardRotationFunc!() -> (then, DeltaTime:double=~) {is=title K2Node_FunctionEntry}
   n2 = Get RotationAlpha() -> (RotationAlpha:double)
   > Set RotationAlpha(RotationAlpha:double=float + float(A=float * float(A=@n1, B=0.8), B=@n2)) -> (Output_Get:double)
   > Set!(execute, CardRotation_Roll:float=FInterpEaseInOut(A=0, B=180, Alpha=@n2, Exponent=3), CardRotation_Pitch:float=0, CardRotation_Yaw:float=0) -> (then, Output_Get:Rotator)
   > Branch(Condition="float >= float"(A=@n2, B=1))
   n3 = Get ReverseRotationAlpha() -> (ReverseRotationAlpha:double)
   > Set ReverseRotationAlpha(ReverseRotationAlpha:double=float + float(A=float * float(A=@n1, B=0.8), B=@n3)) -> (Output_Get:double)
   > Set(CardRotation_Roll=FInterpEaseInOut(A=180, B=360, Alpha=@n3, Exponent=3), CardRotation_Pitch=0, CardRotation_Yaw=0)
   > Branch(Condition="float >= float"(A=@n3, B=1))
   > Set CardRotation(CardRotation:Rotator="0, 0, 0") -> (Output_Get:Rotator)
   > Set RotationAlpha(RotationAlpha=0)
   > Set ReverseRotationAlpha(ReverseRotationAlpha=0)
  ]}
 EventGraph {nodes [
   n1 = Event BlueprintProcessAnimInputEvent()
   > Branch(Condition=IsNamed(self=@n1.AnimInputEvent, InEventName=aie.ChestOpened))
   > Set isChestOpened?(isChestOpened?=true)
   Event BlueprintInitializeAnimation()
   > n2 = Cast To PlayerPawn_Athena(Object=TryGetPawnOwner())
   > Assign On Player Looted Container!(
    execute:Exec
    self:Fort Player Pawn Athena Object Reference=@n2
    Delegate:"Delegate (by ref)"=CustomEvent!() -> (OutputDelegate, then, FortPlayerPawnAthena:Fort Player Pawn Athena Object Reference, LootedContainer:Building Container Object Reference) {is=title K2Node_CustomEvent}
   ) -> (then:Exec) {is=title K2Node_AssignDelegate}
   Event BlueprintPostEvaluateAnimation() {is=title K2Node_Event}
   > Set isChestOpened?(isChestOpened?=false)
   Branch(Condition=Get isChestOpened?())
   > n3 = Sequence()
   n4 = Event BlueprintUpdateAnimation()
   CardMovementFunc!(execute=@n3.then_0, self, DeltaTime:double=@n4.DeltaTimeX) -> (then) {is=title K2Node_CallFunction}
   CardRotationFunc!(execute=@n3.then_1, self, DeltaTime:double=@n4.DeltaTimeX) -> (then) {is=title K2Node_CallFunction}
   Set isChestOpened?(isChestOpened?=true)
   > Delay(Duration=0.1)
   > Set isChestOpened?(isChestOpened?=false)
  ]}
]
variables [
 AnimTimer {type=number}
 CardMovement {type=Vector}
 CardRotation {type=Rotator}
 isChestOpened? {type=boolean}
 MovementAlpha {type=number}
 ReverseMovementAlpha {type=number}
 ReverseRotationAlpha {type=number}
 RotationAlpha {type=number}
]
